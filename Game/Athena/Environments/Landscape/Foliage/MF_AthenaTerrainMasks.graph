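# Material function that samples a terrain biome mask texture and outputs the packed RGBA plus separate Tundra, GrassLand, Desert and unused alpha masks.
# The texture comes from the BiomeTexture parameter, defaulting to the Asteria biome map, and is read with a shared wrap sampler.
# The UseLandscapeCoords static switch selects landscape layer coordinates or, by default, the Adjustable output of MF_Apollo_Terrain_Coords.
# Named reroutes carry the coordinates and texture to the single sample.
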MaterialFunction /Game/Athena/Environments/Landscape/Foliage/MF_AthenaTerrainMasks
graph {nodes [
  n1 = TextureSample(UVs=NamedRerouteUsage(), Tex=NamedRerouteUsage()) {
   desc=BiomeMasks
   samplerSource=SSM_Wrap_WorldGroupSettings
   samplerType=SAMPLERTYPE_LinearColor
   texture=/Game/Athena/Apollo/Environments/Landscape/Materials/Biome/CommonMaterials/T_ApolloBiome2
  }
  FunctionOutput(None=@n1.RGBA) {outputName=BiomeMaskRGBA, sortPriority=-6}
  FunctionOutput(None=@n1.R) {outputName=Tundra, sortPriority=-5}
  FunctionOutput(None=@n1.G) {outputName=GrassLand, sortPriority=-4}
  FunctionOutput(None=@n1.B) {outputName=Desert, sortPriority=-3}
  FunctionOutput(None=@n1.A) {outputName=NotInUse, sortPriority=-2}
  n2 = StaticBool()
  NamedRerouteDeclaration(None=StaticSwitch(
    True=LandscapeLayerCoords() {mappingScale=2033}
    False=MaterialFunctionCall!(WorldPosition) {materialFunction=/Game/Athena/Environments/Landscape/MaterialFunctions/Utilities/MF_Apollo_Terrain_Coords}.Adjustable
    Value=FunctionInput(Preview=@n2) {bUsePreviewValueAsDefault=true, inputName=UseLandscapeCoords, inputType=FunctionInput_StaticBool}
   )) {name=BaseCoord, nodeColor=<rgba 1 0 1 1>}
  n3 = TextureObjectParameter() {
   group=BiomeSettings
   parameterName=BiomeTexture
   samplerType=SAMPLERTYPE_LinearColor
   texture=/Game/Athena/Asteria/Environments/Landscape/Materials/Textures/T_AsteriaBiome
  }
  NamedRerouteDeclaration(None=@n3) {name=BiomeTexture, nodeColor=<rgba 1 0 1 1>}
 ]}
