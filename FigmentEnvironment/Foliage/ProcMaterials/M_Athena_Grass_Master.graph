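# Master grass material for Athena landscape foliage: masked, two-sided subsurface, usable with instanced meshes and Nanite.
# World position offset blends noise-panned wind with storm response, player bending and physics-field pushes, masked by UV1.
# It also sinks grass by instance fade and camera proximity, with an optional legacy snow drop.
# Biome masks and gradient tip tints drive color, subsurface fades with depth, and normals blend instance up with flattened texture detail.
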Material /FigmentEnvironment/Foliage/ProcMaterials/M_Athena_Grass_Master
graph {
 properties {
  bTangentSpaceNormal=false
  bUseMaterialAttributes=true
  bUsedWithInstancedStaticMeshes=true
  bUsedWithNanite=true
  blendMode=BLEND_Masked
  pixelDepthOffsetMode=PDOM_Legacy
  shadingModel=MSM_Subsurface
  twoSided=true
  usedShadingModels=Subsurface
 }
 nodes [
  n1 = NamedRerouteUsage()
  n2 = ComponentMask(None=TextureCoordinate() {coordinateIndex=1}) {g=true}
  n3 = ComponentMask(None=WorldPosition().XYZ) {g=true, r=true}
  n4 = TextureSample(UVs=Divide(A=@n3, B=ScalarParameter() {defaultValue=1000, group=WPO, parameterName=HeightVarianceScale})) {samplerType=SAMPLERTYPE_LinearColor, texture=/Game/Environments/Foliage/Textures/T_FieldGrass_01_M}
  n5 = LinearInterpolate(
   A=ScalarParameter() {group=WPO, parameterName=StaticWPOLow}
   B=ScalarParameter() {defaultValue=30, group=WPO, parameterName=StaticWPOHigh}
   Alpha=@n4.R
  )
  n6 = PerInstanceFadeAmount()
  n7 = Multiply(
   A=Multiply(
    A=@n4.R
    B=CollectionParameter() {collection=/Game/Athena/Environments/Landscape/MPC/MPC_Landscape_S07, parameterName=Ambient_WindMotion_Amount}
   )
   B=Constant() {r=300}
  )
  n8 = Divide(A=@n3, B=ScalarParameter() {defaultValue=1200, group=WPO, parameterName=WindWaveScale})
  n9 = MaterialFunctionCall() {materialFunction=/Game/Athena/Environments/Landscape/MaterialFunctions/MF_StormMask}
  n10 = Multiply(
   A=Time()
   B=LinearInterpolate(
    A=CollectionParameter() {collection=&MPC_Landscape_S07, parameterName=Grass_WindSpeed}
    B=CollectionParameter() {collection=&MPC_Landscape_S07, parameterName=Grass_WindSpeed_Storm}
    Alpha=@n9.Result
   )
  )
  n11 = Multiply(
   A=TextureSample(UVs=Panner(Coordinate=@n8, Time=@n10) {speedY=1}) {samplerType=SAMPLERTYPE_Normal, texture=/Game/Global/Textures/Default/Noise/T_SwirlNoise_N}.RGB
   B=LinearInterpolate(
    A=CollectionParameter() {collection=&MPC_Landscape_S07, parameterName=Grass_WindPower}
    B=CollectionParameter() {collection=&MPC_Landscape_S07, parameterName=Grass_WindPower_Storm}
    Alpha=@n9.Result
   )
  )
  n12 = Normalize(VectorInput=AppendVector(A=ComponentMask(None=@n11) {g=true, r=true}, B=Constant() {r=1}))
  n13 = Add(
   A=Multiply(A=Constant3Vector() {constant=<rgba 0 0 1 0>}, B=Multiply(A=@n5, B=@n6))
   B=Multiply(A=Multiply(A=@n7, B=@n12), B=Constant3Vector() {constant=<rgba 1 1 0 0>})
  )
  n14 = NamedRerouteUsage()
  n15 = ScalarParameter() {defaultValue=1, group=Animation, parameterName=MaxRotation_PhysicsField}
  n16 = Multiply(
   A=Multiply(A=Multiply(A=@n14, B=@n15), B=NamedRerouteUsage())
   B=ScalarParameter() {defaultValue=86.75, parameterName=Explosion Strenth modifier}
  ) {constB=86.75}
  n17 = NamedRerouteUsage()
  n18 = CollectionParameter() {collection=&FortniteMaterialParameters, parameterName=PlayerPosition}
  n19 = MaterialFunctionCall!(
   PlayerPosition=@n18
   PlayerRadiusMax=CollectionParameter() {collection=&MPC_Landscape_S07, parameterName=PlayerBend_Radius}
   BendStrength=CollectionParameter() {collection=&MPC_Landscape_S07, parameterName=PlayerBend_Strength}
  ) {materialFunction=/Game/Athena/Environments/Landscape/MaterialFunctions/MF_AthenaFoliagePlayerInteract}
  n20 = Multiply(A=OneMinus(None=@n2), B=LinearInterpolate(A=LinearInterpolate(A=@n13, B=@n16, Alpha=@n17), B=@n19.Rotation, Alpha=@n19.Rotation))
  n21 = Constant2Vector()
  n22 = Multiply(A=OneMinus(None=PerInstanceFadeAmount()), B=ScalarParameter() {defaultValue=-65, parameterName=SinkIntoGroundDist})
  n23 = Multiply(
   A=SphereMask(
    A=CameraPositionWS()
    B=WorldPosition().XYZ
    Radius=ScalarParameter() {defaultValue=200, parameterName=camerafadeMaskRadius}
    Hardness=ScalarParameter() {parameterName=camerafadeMaskHardness}
   )
   B=Constant() {r=-200}
  )
  n24 = Constant()
  n25 = CollectionParameter() {collection=&FortniteMaterialParameters, parameterName=IsPlayerSliding}
  n26 = Add(
   A=Add(A=QualitySwitch(Default=@n20, High=@n20, Medium=@n20), B=AppendVector(A=@n21, B=@n22))
   B=AppendVector(A=Constant2Vector(), B=LinearInterpolate(A=@n23, B=@n24, Alpha=@n25))
  )
  n27 = ScalarParameter() {defaultValue=-50, group=WPO-DistanceBased, parameterName=SinkIntoGroundAmount}
  n28 = Constant()
  n29 = PerInstanceFadeAmount()
  n30 = Add(A=@n26, B=Multiply(A=LinearInterpolate(A=@n27, B=@n28, Alpha=@n29), B=Constant3Vector() {constant=<rgba 0 0 1 0>})) {constB=0}
  n31 = Add(A=WorldPosition() {worldPositionShaderOffset=WPT_ExcludeAllShaderOffsets}.XYZ, B=@n30)
  n32 = MaterialFunctionCall!(
   UseHeightLerp
   UV_WorldMask
   Use_ExtraSnowMask
   LandscapeBlend_Height
   Landscape_RockMask
   Snow_Mask
   Simple_UseWorldZNormal
   Simple_WorldNormal
   WorldPosition
  ) {materialFunction=/Game/Athena/Environments/Landscape/MaterialFunctions/Snow/MF_Athena_Holiday_SnowFall_S7}
  n33 = OneMinus(None=Power(Base=OneMinus(None=@n32.SnowMask)))
  Add(A=@n30, B=Multiply(A=Subtract(A=@n1, B=@n31), B=@n33))
  n34 = TextureCoordinate() {coordinateIndex=2}
  n35 = MaterialFunctionCall(X=ComponentMask(None=@n34) {r=true}, Y=ComponentMask(None=@n34) {g=true}, Z=Constant()) {materialFunction=&MakeFloat3}
  Add(A=@n35.Result)
  n36 = Subtract(A=WorldPosition().XYZ, B=CameraPositionWS())
  n37 = Constant3Vector() {constant=<rgba 1 1 0 0>}
  Add(
   A=Multiply(A=Multiply(A=@n36, B=@n37), B=ScalarParameter() {defaultValue=1, parameterName=CameraFadeMaskPush})
   B=AppendVector(A=Constant2Vector(), B=ScalarParameter() {defaultValue=-200, parameterName=CameraFadeMaskSink})
  )
  AppendVector(A=Constant2Vector() {g=1, r=1}, B=TwoSidedSign())
  n38 = ComponentMask(None=@n26) {b=true}
  AppendVector(A=ComponentMask(None=@n26) {g=true, r=true}, B=@n38)
  Constant() {r=100}
  Desaturation(Fraction=ScalarParameter() {parameterName=testingGrass})
  n39 = WorldPosition()
  n40 = MaterialFunctionCall() {materialFunction=&ObjectPivotPoint}
  LinearInterpolate(A=Normalize(VectorInput=Subtract(A=@n39.XYZ, B=@n40.Object Pivot Location)), B=Constant3Vector() {constant=<rgba 0 0 1 0>})
  n41 = TextureSampleParameter2D() {parameterName=Gradient_Mask, samplerType=SAMPLERTYPE_LinearColor, texture=/Game/Athena/Prototype/Terrain/T_GrassGradient_01}
  n42 = Power(Base=@n41.G, Exp=Constant() {r=3})
  n43 = Multiply(A=Constant3Vector() {constant=<rgba 0 0 -100 0>}, B=@n33)
  MakeMaterialAttributes!(
   BaseColor=NamedRerouteUsage()
   Metallic
   Specular=LinearInterpolate(A=ScalarParameter() {defaultValue=0.02, parameterName=SpecMin}, B=ScalarParameter() {defaultValue=0.2, parameterName=Spec}, Alpha=@n42)
   Roughness=ShadingPathSwitch(Default=ScalarParameter() {defaultValue=0.6, parameterName=Rough}, Mobile=Constant() {r=1}) {desc=Migrated from FeatureLevelSwitch}
   Anisotropy
   EmissiveColor
   Opacity=NamedRerouteUsage()
   OpacityMask=TextureSampleParameter2D() {parameterName=Grass_Mask, texture=/Game/Environments/Foliage/Textures/T_WavyGrass_01}.R
   Normal=NamedRerouteUsage()
   Tangent
   WorldPositionOffset=StaticSwitchParameter(True=Add(A=@n30, B=@n43) {desc="just move down for nanite as pivots aren't per blade (it looks like)"}, False=@n30) {group=Snow, parameterName=UseSnowLegacy}
   SubsurfaceColor=NamedRerouteUsage()
   ClearCoat
   ClearCoatRoughness
   AmbientOcclusion
   Refraction
   CustomizedUVs_0
   CustomizedUVs_1
   CustomizedUVs_2
   CustomizedUVs_3
   CustomizedUVs_4
   CustomizedUVs_5
   CustomizedUVs_6
   CustomizedUVs_7
   PixelDepthOffset
   ShadingModel
   Displacement
  )
  n44 = ScalarParameter() {parameterName=Zoffset}
  Multiply(A=@n38, B=@n44)
  Multiply(B=ComponentMask(None=CollectionParameter() {collection=/FigmentCore/Game/Athena/Environments/Landscape/MPC/MPC_Landscape, parameterName=Arid_Grass_Tint}) {b=true, g=true, r=true})
  Multiply(B=VectorParameter() {parameterName=Arid_SandGrass_Tint, defaultValue=<rgba 0.4678 0.314 0.1384 1>}.RGB)
  n45 = MaterialFunctionCall!(UseLandscapeCoords) {materialFunction=/Game/Athena/Environments/Landscape/Foliage/MF_AthenaTerrainMasks}
  n46 = VectorParameter() {parameterName=GrassSample_Tint, defaultValue=<rgba 0.9 0.9 0.9 0>}
  n47 = StaticSwitchParameter(
   True=@n46.RGB
   False=LinearInterpolate(A=@n46.RGB, B=VectorParameter() {parameterName=GrassSample_Tint_Arid, defaultValue=<rgba 0.9 0.9 0.9 0>}.RGB)
  ) {parameterName=ForceAridGrass}
  n48 = VectorParameter() {parameterName=GrassTipTint, defaultValue=<rgba 1.1 1.1 1.1 1>}
  n49 = ShadingPathSwitch(Default=@n48.RGB, Mobile=Multiply(A=@n48.RGB) {constB=1.2}) {desc=Migrated from FeatureLevelSwitch}
  n50 = Clamp(None=Multiply(
    A=MaterialFunctionCall!(BiomeMask-RGBA=@n45.BiomeMaskRGBA, Grass Localised Tint Mask, CombinedLandscapeLayersMask, BiomeColors-UseDefault) {materialFunction=/Game/Athena/Asteria/Environments/Landscape/Materials/MaterialFunctions/MF_Asteria_Common_Grass}.BasicGrassMeshColor
    B=LinearInterpolate(A=@n47, B=@n49, Alpha=@n42)
   )) {clampMode=CMODE_ClampMax}
  NamedRerouteDeclaration(None=@n50) {name=Base Color, nodeColor=<rgba 0 0.7422 1 1>}
  n51 = NamedRerouteUsage()
  n52 = ScalarParameter() {defaultValue=0.1, desc=Multi to control sss color strength, group=RufusSubSurface, parameterName=SSS_Multi}
  n53 = Divide(
   A=Add(A=PixelDepth().R, B=ScalarParameter() {defaultValue=-7000, group=Camera Fade Distances, parameterName=FarFadeOffset})
   B=ScalarParameter() {defaultValue=4000, group=Camera Fade Distances, parameterName=FarFade}
  )
  NamedRerouteDeclaration(None=LinearInterpolate(A=Multiply(A=@n51, B=@n52), Alpha=Saturate(None=@n53)) {constB=0}) {name=SSS Color, nodeColor=<rgba 0.5703 1 0 1>}
  n54 = Transform(None=Constant3Vector() {desc=Looks almost same RVT normal, constant=<rgba 0 0 1 0>}) {transformSourceType=TRANSFORMSOURCE_Instance}
  n55 = TextureSample() {samplerType=SAMPLERTYPE_Normal, texture=/Game/Environments/Foliage/Textures/T_WavyGrass_01_N}
  n56 = ScalarParameter() {defaultValue=0.84, group=Normal Settings, parameterName=TextureNormalFlatten}
  NamedRerouteDeclaration(None=MaterialFunctionCall(
    BaseNormal=VertexInterpolator(VS=@n54).PS
    AdditionalNormal=MaterialFunctionCall(Normal=@n55.RGB, Flatness=@n56) {materialFunction=&FlattenNormal}.Result
   ) {materialFunction=&BlendAngleCorrectedNormals}.Result) {name=normal map, nodeColor=<rgba 0.9219 0 1 1>}
  n57 = ScalarParameter() {defaultValue=1, group=SubSurface Settings, parameterName=SSSOpacity}
  n58 = ScalarParameter() {defaultValue=0.5, group=SubSurface Settings, parameterName=SSSOpacity2}
  n59 = MaterialFunctionCall() {materialFunction=&LinearGradient}
  n60 = Subtract(A=@n59.VGradient, B=ScalarParameter() {defaultValue=0.624, group=SubSurface Settings, parameterName=SSS gradient adjust})
  NamedRerouteDeclaration(None=Saturate(None=LinearInterpolate(A=@n57, B=@n58, Alpha=@n60))) {name=SSSOpacity, nodeColor=<rgba 0.1719 0 1 1>}
  n61 = SamplePhysicsVectorField!(World Position=NamedRerouteDeclaration(None=TransformPosition(None=@n35.Result) {transformSourceType=TRANSFORMPOSSOURCE_Instance, transformType=TRANSFORMPOSSOURCE_World}) {name=PivotPosition, nodeColor=<rgba 0 0.9766 1 1>}) {fieldTarget=Vector_LinearVelocity}
  n62 = MaterialFunctionCall(Vector 3=Add(A=@n61) {constB=0.0001}) {materialFunction=/Engine/Functions/Engine_MaterialFunctions02/Utility/VectorLength}
  n63 = Constant()
  NamedRerouteDeclaration(None=ShadingPathSwitch(Default=Saturate(None=@n62.V3 Length), Mobile=@n63) {desc=Migrated from FeatureLevelSwitch}) {name=PhysicsFieldMask, nodeColor=<rgba 0 1 0.3906 1>}
  n64 = Constant3Vector() {constant=<rgba 0 0 1 0>}
  NamedRerouteDeclaration(None=ShadingPathSwitch(
    Default=MaterialFunctionCall(Vector=@n61, Default=@n64) {materialFunction=&SafeNormalize}.Result
    Mobile=Constant3Vector() {constant=<rgba 0 0 1 0>}
   ) {desc=Migrated from FeatureLevelSwitch}) {name=PhysicsFieldDirection, nodeColor=<rgba 0.3828 0 1 1>}
  n65 = Min(A=@n62.V3 Length, B=ScalarParameter() {defaultValue=1, group=Animation, parameterName=MaxRotation_PhysicsField_Clamp})
  n66 = Time()
  n67 = ScalarParameter() {defaultValue=351, group=Animation, parameterName=PhysicsField_NoiseOffset}
  n68 = Frac(None=Multiply(A=ScalarParameter() {defaultValue=2.5, group=Animation, parameterName=MaxRotation_PhysicsField_NoiseSpeed}, B=Add(A=@n66, B=@n67)))
  n69 = OneMinus(None=@n68)
  n70 = LinearInterpolate(
   A=Multiply(A=Multiply(A=@n68, B=@n69)) {constB=2}
   Alpha=ScalarParameter() {defaultValue=0.6, group=Animation, parameterName=MaxRotation_PhysicsField_MInBend}
  )
  NamedRerouteDeclaration(None=ShadingPathSwitch(Default=Multiply(A=@n65, B=@n70), Mobile=@n63) {desc=Migrated from FeatureLevelSwitch}) {name=PhysicsFieldStrength, nodeColor=<rgba 1 0 0.2109 1>}
  Reroute(None=@n59.VGradient)
  ScalarParameter() {defaultValue=-50, parameterName=PlayerBendStrength}
  ScalarParameter() {defaultValue=200, parameterName=PlayerBendRadius_1}
  ScalarParameter() {defaultValue=0.08, group=WPO, parameterName=WindSpeed}
  ScalarParameter() {defaultValue=0.75, group=WPO, parameterName=WindSpeed_1}
  ScalarParameter() {defaultValue=0.6, group=WPO, parameterName=WindPower}
  ScalarParameter() {defaultValue=1, group=WPO, parameterName=WindPower_Storm}
  ScalarParameter() {defaultValue=0.4, parameterName=Ambient Wind Motion Amount}
  n71 = VertexNormalWS()
  n72 = LinearInterpolate(
   A=Transform(None=TextureSample() {samplerType=SAMPLERTYPE_Normal, texture=&T_WavyGrass_01_N}.RGB)
   B=@n71
   Alpha=Clamp(None=@n2)
  )
  ShadingPathSwitch(Default=QualitySwitch(Default=@n72, Low=@n71, High=@n72, Medium=@n72), Mobile=@n71) {desc=Migrated from FeatureLevelSwitch}
  n73 = ScalarParameter() {defaultValue=0.85, parameterName=ES2_FakeAODarkness}
  ShadingPathSwitch(Default=@n50, Mobile=Multiply(A=@n50, B=LinearInterpolate(A=@n73, Alpha=@n41.G))) {desc=Migrated from FeatureLevelSwitch}
  ShadingPathSwitch(Default=@n26, Mobile=Constant()) {desc=Migrated from FeatureLevelSwitch}
  Subtract(A=@n38, B=@n44)
  VertexColor()
 ]
}
